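; Guided-motion movement mode for the Rift Gauntlets Ascend, launching the pawn upward and along its movement input.
; Direction blends normalized acceleration with up by XY/Z force ratio; strength sums both forces, scaled by a consecutive-use penalty from RecentlyUsed effect stacks.
; Root-motion Z velocity is clamped through the RiftGauntlets pawn component.
; Activation fires a local cue and listens for PreLaunch events, which end the mode and optionally cap Z velocity via console variable.
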(event EventOnExtensionActivate
  (params (RuntimeContext :type MMERuntime%20Context%20Structure%20%28by%20ref%29))
  (bind (_velocity_x _velocity_y _velocity_z) (Math|Vector|BreakVector (Variables|/Script/Engine.MovementComponent|GetVelocity_X (Variables|FortMovementMode_BaseExtLogic|GetOwnerMovementComponent))))
  ;[0.01374 0.15 0.1201 0.5] OnExtensionActivate/Deactivate: LaunchCharacter knockbacks should cancel the MME.
  (|CallParentFunction RuntimeContext)
  (CallFunction|ExecuteGameplayCueNotReplicated (Utilities|Struct|BreakGameplayCueTag (Variables|Default|GetActivateGameplayCueTag)))
  (bind _returnvalue (Class|FortMovementModeExtBlueprintLibrary|GetRuntimeDatafromContext RuntimeContext "/Script/FortniteGame.FortMovementMode_BaseExtRuntimeData"))
  (Variables|Transient|SetRuntimeData _returnvalue)
  (bind _ownerpawn (Variables|FortMovementMode_BaseExtLogic|GetOwnerPawn)
    (:then
      (bind _asynctaskproxy (Ability|Async|WaitGameplayEventToActor _ownerpawn "(TagName=\"Event.Character.PreLaunch\")")
        (:then
          (Variables|Transient|SetWaitGameplayEventAction _asynctaskproxy))
        (:EventReceived
          (bind _runtimedata (Variables|Transient|GetRuntimeData)
            (:then
              (Class|FortMovementModeBaseExtRuntimeData|ForceManualEnd _runtimedata)
              ; Clamp velocity Z when going to be launched, in case we have weirdly large Z speed from calculation
              (bind _returnvalue_2 (Development|GetConsoleVariableFloatValue "RiftGauntlets.AscendLaunch.ClampZVelocity"))
              (if (> _returnvalue_2 0)
                (Variables|/Script/Engine.MovementComponent|SetVelocity_X (Math|Vector|MakeVector _velocity_x _velocity_y (Math|Float|Min(Float) _velocity_z _returnvalue_2)) (Variables|FortMovementMode_BaseExtLogic|GetOwnerMovementComponent))))
            (:else)))))
    (:else)))

(event EventOnExtensionDeactivate
  (params (RuntimeContext :type MMERuntime%20Context%20Structure%20%28by%20ref%29))
  ;[0.01374 0.15 0.1201 0.5] OnExtensionActivate/Deactivate: LaunchCharacter knockbacks should cancel the MME.
  (|CallParentFunction RuntimeContext)
  (bind _waitgameplayeventaction (Variables|Transient|GetWaitGameplayEventAction)
    (:then
      (Ability|Async|EndAction _waitgameplayeventaction))
    (:else)))

(fn BP_AdjustRootMotionVelocity
  (params (RuntimeContext :type MMERuntime%20Context%20Structure%20%28by%20ref%29) (InVelocity :type Vector :default "0, 0, 0"))
  ;[0.04224 0.121 0.15 0.5] AdjustRootMotionVelocity: Overridden to clamp the velocity on the Z-axis based on height above ground / height above build grid.
  (Utilities|FlowControl|Sequence
    (:then_0
      (bind _ownerpawn (Variables|FortMovementMode_BaseExtLogic|GetOwnerPawn)
        (:then
          (bind _returnvalue (Actor|GetComponentbyClass _ownerpawn "/RiftGauntletsGameplay/Gameplay/B_RiftGauntletsPawnComponent.B_RiftGauntletsPawnComponent_C"))
          (Utilities|IsValid _returnvalue
            (:Is%20Valid
              (bind _output_get (LocalVariables|SetVelocityZ (.z InVelocity)))
              (bind _returnvalue_2 (Class|FortPawnComponentRiftGauntlets|PerformMovementModeClampMaxVelocityZ _returnvalue _output_get))
              (if _returnvalue_2
                (return (Math|Vector|MakeVector (.x InVelocity) (.y InVelocity) _output_get))))))
        (:else)))
    (:then_1
      (return InVelocity))))

(fn GetDirection_ConstantForce
  (params (GuidedMotionRuntimeData :type Fort%20Movement%20Mode%20Guided%20Motion%20Runtime%20Data%20Object%20Reference))
  (return (.x (* (Math|Vector|Normalize (Pawn|Components|CharacterMovement|GetCurrentAcceleration (Variables|FortMovementMode_BaseExtLogic|GetOwnerMovementComponent))) (Math|Float|SafeDivide (ScalableFloat|GetValueatLevel (Variables|Default|GetHF_XYLaunchForce)) (+ (ScalableFloat|GetValueatLevel (Variables|Default|GetHF_XYLaunchForce)) (ScalableFloat|GetValueatLevel (Variables|Default|GetHF_ZLaunchForce)))))) (.y (* (Math|Vector|Normalize (Pawn|Components|CharacterMovement|GetCurrentAcceleration (Variables|FortMovementMode_BaseExtLogic|GetOwnerMovementComponent))) (Math|Float|SafeDivide (ScalableFloat|GetValueatLevel (Variables|Default|GetHF_XYLaunchForce)) (+ (ScalableFloat|GetValueatLevel (Variables|Default|GetHF_XYLaunchForce)) (ScalableFloat|GetValueatLevel (Variables|Default|GetHF_ZLaunchForce)))))) (Math|Float|SafeDivide (ScalableFloat|GetValueatLevel (Variables|Default|GetHF_ZLaunchForce)) (+ (ScalableFloat|GetValueatLevel (Variables|Default|GetHF_XYLaunchForce)) (ScalableFloat|GetValueatLevel (Variables|Default|GetHF_ZLaunchForce))))))

(fn GetStrength_ConstantForce
  (params (GuidedMotionRuntimeData :type Fort%20Movement%20Mode%20Guided%20Motion%20Runtime%20Data%20Object%20Reference))
  (Variables|Config-ConstantForce|SetStrength (+ (ScalableFloat|GetValueatLevel (Variables|Default|GetHF_XYLaunchForce)) (ScalableFloat|GetValueatLevel (Variables|Default|GetHF_ZLaunchForce))))
  (Utilities|FlowControl|Sequence
    (:then_0
      (bind _ownerpawn (Variables|FortMovementMode_BaseExtLogic|GetOwnerPawn)
        (:then
          (Utilities|IsValid (Ability|GetFortAbilitySystemComponent _ownerpawn)
            (:Is%20Valid
              (bind _returnvalue (GameplayEffects|GetActiveEffectswithAllTags (Ability|GetFortAbilitySystemComponent _ownerpawn) "(GameplayTags=((TagName=\"RiftGauntlets.RecentlyUsed.Ability.Ascend\")))"))
              (for _array_element _returnvalue
                (bind _returnvalue_2 (Ability|GameplayEffect|GetActiveGameplayEffectStackCount _array_element))
                (LocalVariables|SetStackCount (+ _returnvalue_2 (LocalVariables|GetStackCount))))
              (if (> (LocalVariables|GetStackCount) 0)
                (LocalVariables|SetMultiplier (ScalableFloat|GetValueatLevel (Utilities|Array|Get(aref) (Utilities|Array|MakeArray (Variables|Default|GetHF_ConsecutivePenaltyMultiplier_1) (Variables|Default|GetHF_ConsecutivePenaltyMultiplier_2) (Variables|Default|GetHF_ConsecutivePenaltyMultiplier_3) (Variables|Default|GetHF_ConsecutivePenaltyMultiplier_4)) (- (LocalVariables|GetStackCount) 1))))))))
        (:else)))
    (:then_1
      (return (* (Variables|Config-ConstantForce|GetStrength) (LocalVariables|GetMultiplier))))))
